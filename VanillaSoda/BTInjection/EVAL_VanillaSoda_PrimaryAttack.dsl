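; Behavior tree evaluator, named for the VanillaSoda primary attack, that relies on its parent class for evaluation logic.
; On enter, it resolves the behavior tree component's owner to a Controller, casts the controlled pawn to FortPawn and caches it in a variable.
; Its starting-conditions check only forwards to the parent function and returns that result, adding no conditions of its own.
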(event EventOnEnter
  (params (OwnerComp :type Behavior%20Tree%20Component%20Object%20Reference))
  (bind _ascontroller (Utilities|Casting|CastToController (Components|GetOwner OwnerComp)))
  (bind _asfort_pawn (Utilities|Casting|CastToFortPawn (Pawn|GetControlledPawn _ascontroller)))
  (Variables|Default|SetFortPawn _asfort_pawn))

(fn EvaluateStartingConditions
  (params (OwnerComp :type Behavior%20Tree%20Component%20Object%20Reference))
  (bind _returnvalue (|CallParentFunction OwnerComp))
  (return _returnvalue))
